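# Physics asset for the BurnishedRaven vehicle cosmetic body, defining its simplified collision shape.
# A single box body on the root bone approximates the whole vehicle, offset slightly along X and raised above the bone origin.
# It has no constraints and no other bodies, so the body is one rigid piece with no articulated parts.
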PhysicsAsset /VehicleCosmetics/Bodies/BurnishedRaven/Meshes/Body_BurnishedRaven_PhysicsDM
bodies {root [{shapeType=Box, extentX=444, extentY=216, extentZ=116, shapeName=root_box, center=<10.6954 0 99.0732>}]}
constraints []
